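# Material function for the Octane GherkinLTE car body skin that animates falling or scrolling, spinning "cheto" sprites over an orange base.
# The base blends flat orange with a vertical gradient using the skin texture as mask.
# Input UVs are tiled, panned, wrapped with Frac and rotated to repeat two sprite variants per cell.
# Channels of a slower-panning random mask gate the sprites, and a faster, offset layer composites on top for variation.
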MaterialFunction /VehicleCosmetics/Bodies/Octane/Skins/OctaneGherkinLTE/MF_Octane_GhekinLTE
graph {nodes [
  n1 = Constant3Vector() {constant=<rgba 0.673 0.173 0.0012 0>}
  n2 = ComponentMask(None=TextureCoordinate()) {g=true}
  n3 = TextureSample() {texture=&/Skin_Octane_GherkinLTE_RGB}
  n4 = LinearInterpolate(A=@n1, B=Multiply(A=OneMinus(None=@n2), B=@n1), Alpha=OneMinus(None=ComponentMask(None=@n3.R) {g=true}))
  n5 = FunctionInput() {inputType=FunctionInput_Vector2}
  n6 = Subtract(A=Multiply(A=@n5, B=Constant() {r=8}), B=Constant2Vector())
  n7 = Rotator(Coordinate=Frac(None=Panner(Coordinate=@n6) {speedY=1.5})) {speed=-2.5}
  n8 = TextureSample(UVs=@n7) {samplerSource=SSM_Clamp_WorldGroupSettings, texture=&/single_cheto_M}
  n9 = TextureSample(UVs=@n7) {samplerSource=SSM_Clamp_WorldGroupSettings, texture=&/single_cheto}
  n10 = Constant2Vector() {g=1, r=1}
  n11 = TextureSample(UVs=Panner(Coordinate=Subtract(A=@n5, B=@n10)) {speedY=0.1875}) {texture=&/Skin_Octane_GherkinLTE_RandomMask_RGB}
  n12 = Multiply(A=@n9.A, B=@n11.R)
  n13 = Multiply(A=@n8.A, B=@n11.G)
  n14 = LinearInterpolate(A=@n4, B=LinearInterpolate(A=@n8.RGB, B=@n9.RGB, Alpha=@n12), Alpha=Clamp(None=Add(A=@n12, B=@n13)))
  n15 = TextureSample(UVs=@n7) {texture=&single_cheto_M}
  n16 = TextureSample(UVs=@n7) {texture=&single_cheto}
  n17 = Multiply(A=@n16.A, B=@n11.B)
  n18 = Multiply(A=@n15.A, B=@n11.A)
  n19 = LinearInterpolate(A=@n4, B=LinearInterpolate(A=@n15.RGB, B=@n16.RGB, Alpha=@n17), Alpha=Clamp(None=Add(A=@n17, B=@n18)))
  n20 = Subtract(A=Multiply(A=@n5, B=Constant() {r=8}), B=Constant2Vector() {g=1, r=1.5})
  n21 = Rotator(Coordinate=Frac(None=Panner(Coordinate=@n20) {speedY=2.4})) {speed=-3}
  n22 = TextureSample(UVs=@n21) {samplerSource=SSM_Clamp_WorldGroupSettings, texture=&single_cheto_M}
  n23 = TextureSample(UVs=@n21) {samplerSource=SSM_Clamp_WorldGroupSettings, texture=&single_cheto}
  n24 = Constant2Vector() {r=0.065}
  n25 = TextureSample(UVs=Panner(Coordinate=Subtract(A=@n5, B=@n24)) {speedY=0.3}) {texture=&Skin_Octane_GherkinLTE_RandomMask_RGB}
  n26 = Multiply(A=@n23.A, B=@n25.R)
  n27 = Add(A=@n26, B=Multiply(A=@n22.A, B=@n25.G))
  FunctionOutput(None=LinearInterpolate(A=LinearInterpolate(A=@n14, B=@n19, Alpha=@n19), B=LinearInterpolate(A=@n22.RGB, B=@n23.RGB, Alpha=@n26), Alpha=Clamp(None=@n27)))
  TextureCoordinate() {coordinateIndex=1, uTiling=1.6, vTiling=1.6}
 ]}
